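; Quest gameplay ability for S12 that tracks whether the owning player is in a Hoagie vehicle.
; It activates from a gameplay event, calls the parent, then switches on the Vehicle Entered or Exit tag.
; On entry it checks the vehicle's ability system component for the Hoagie tag before applying the InHoagie effect to the owner.
; On exit it removes effects granting the InHoagie tag, and the ability then ends.
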(event Ability|EventActivateAbilityFromEvent
  (params (EventData :type Gameplay%20Event%20Data%20Structure%20%28by%20ref%29))
  (|CallParentFunction EventData)
  (switch Utilities|FlowControl|Switch|SwitchonGameplayTag (Utilities|Struct|BreakGameplayEventData EventData)
    (:Event.Vehicle.Entered
      (if (GameplayTags|TagContainer|CompareTagAssetInterfacetoOtherTags (Variables|Default|GetT_Hoagie) (Vehicle|GetVehicleAbilitySystemComponent (Vehicle|GetVehicleInterface (Variables|GAVars|Data|GetOwningPlayerPawn))))
        (Ability|ApplyGameplayEffectToOwner (Variables|Default|GetGE_InHoagie))))
    (:Event.Vehicle.Exit
      (Ability|RemoveGameplayEffectFromOwnerWithGrantedTags (Variables|Default|GetT_InHoagie))))
  (Ability|EndAbility))
